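# Sound cue for a Hope computer typing effect in LimeRock, playing a single typing wave through one Wave Player with no randomization.
# It plays at slightly reduced volume on the Fort_Events sound class and borrows the LimeRock tree attenuation settings.
# It stores per-owner concurrency values with a small voice cap and retrigger time, but the override flag is not enabled and the concurrency set holds only a null entry.
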SoundCue /LimeRock/Assets/Audio/SFX/Computer/SC_LimeRock_100_Hope_Computer_Typing
root=WavePlayer_0
nodes [{id=WavePlayer_0, properties {soundWaveAssetPtr=/LimeRock/Assets/Audio/SFX/Computer/Wav/LimeRock_100_Computer_Typing_01}}]
properties {
 attenuationSettings=/LimeRock/Assets/Audio/Settings/ATT_LimeRock_Tree
 maxDistance=2097152
 soundClassObject=&Fort_Events
 volumeMultiplier=0.8
 concurrencyOverrides {bLimitToOwner=true, maxCount=2, retriggerTime=0.25}
 concurrencySet [null]
}
